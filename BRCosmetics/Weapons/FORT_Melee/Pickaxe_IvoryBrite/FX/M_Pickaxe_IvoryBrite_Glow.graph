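# Unlit, translucent, two-sided glow material for the IvoryBrite pickaxe, usable on Niagara sprites and mesh particles.
# Glow texture times particle color feeds a shared emissive adjustment function.
# Alpha multiplies texture and particle alpha by a view-angle fade that hides edge-on surfaces, with depth fade skipped at Low and Medium quality.
# A deploy wipe mask borrowed from the ClaimReflect pickaxe then multiplies that alpha.
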Material /BRCosmetics/Weapons/FORT_Melee/Pickaxe_IvoryBrite/FX/M_Pickaxe_IvoryBrite_Glow
graph {
 properties {
  bUsedWithNiagaraMeshParticles=true
  bUsedWithNiagaraSprites=true
  blendMode=BLEND_Translucent
  pixelDepthOffsetMode=PDOM_Legacy
  shadingModel=MSM_Unlit
  twoSided=true
  usedShadingModels=Unlit
 }
 nodes [
  n1 = TextureSample() {texture=&/T_Pickaxe_IvoryBrite_Glow}
  n2 = ParticleColor()
  MaterialFunctionCall(Emissive=Multiply(A=@n1.RGB, B=@n2.RGB)) {materialFunction=&MF_FX_EmissiveAdjustments}
  n3 = DotProduct(A=CameraVectorWS(), B=VertexNormalWS())
  n4 = Multiply(
   A=Multiply(A=@n2.A, B=@n1.A)
   B=SmoothStep(
    Min=ScalarParameter() {parameterName=View Angle Fade Min, sliderMax=1}
    Max=ScalarParameter() {defaultValue=0.15, parameterName=View Angle Fade Max, sliderMax=1}
    Value=Abs(None=@n3)
   )
  )
  n5 = DepthFade(Opacity=@n4, FadeDistance=ScalarParameter() {defaultValue=1, parameterName=FadeDistance})
  Multiply(
   A=QualitySwitch(Default=@n5, Low=@n4, High=@n5, Medium=@n4, Epic=@n5)
   B=MaterialFunctionCall!(InDerectionalGradient) {materialFunction=/BRCosmetics/Weapons/FORT_Melee/Pickaxe_ClaimReflect/Materials/MF_ClaimReflect_DeployWipe}.WipeMask
  )
 ]
}
